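; Roadside electronic speed sign that displays the speed of nearby players or vehicles and reacts when they exceed a limit.
; Box overlap starts a server timer that samples the closest tracked actor's speed in km/h, despite a comment saying MPH.
; VehicleSpeed's RepNotify drives a dynamic material, with clients playing a sound and blink on speeding.
; Speeding can also grant quest progress to occupants or walkers, gated by ShouldGrantQuestReward and creative-mode checks.
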(event EventBeginPlay
  (bind _returnvalue (Fortnite|GetLocalFortPlayerControllers true))
  (bind _output (Utilities|Array|Get(acopy) _returnvalue))
  ;[1 1 1] Check if In Creative Mode or not
  (bind _asfort_game_state_athena (Utilities|Casting|CastToFortGameStateAthena (Game|GetGameState)))
  (if (Variables|/Script/FortniteGame.FortGameStateAthena|GetbPlaylistDataIsLoaded _asfort_game_state_athena)
    (bind _cancompletequest (Class|AthenaQuestFunctionLibrary|AthenaCheckCreativeMode _asfort_game_state_athena))
    (if _cancompletequest
      (Variables|Default|SetCreativeMode false)
      (else
        (Variables|Default|SetCreativeMode true)))
    ;[1 1 1] REPLACE
    (Utilities|FlowControl|Delay 5.0)
    (Networking|SwitchHasAuthority
      (:Remote
        ;[1 1 1] Bind Quest Updates
        (bind (_questisvalid _questcompleted _objectivecompleted) (Default|AthenaCheckQuestandObjectiveCompletion _output (Variables|QuestStuff|GetQuestItem) (Variables|QuestStuff|GetObjBackendName)))
        (if _questisvalid
          (if _objectivecompleted
            _
            (elif (Variables|Default|GetCreativeMode)
              _
              (else
                (Quests|BindEventtoOnQuestsUpdated (Quests|GetQuestManager _output "Invalid") (EventDispatchers|CreateEvent :!event Custom|QuestUpdated))
                (Variables|Default|SetQuestLocallyActive true)))))))
    (else
      (Playlist|BindEventtoOnCurrentPlaylistLoaded _asfort_game_state_athena (delegate Custom|PlaylistLoaded)))))

(event Custom|PlaylistLoaded
  (params (PlaylistName :type Name) (PlaylistContextTags :type Gameplay%20Tag%20Container%20Structure%20%28by%20ref%29))
  (bind _returnvalue (Fortnite|GetLocalFortPlayerControllers true))
  (bind _output (Utilities|Array|Get(acopy) _returnvalue))
  ;[1 1 1] Check if In Creative Mode or not
  (bind _cancompletequest (Class|AthenaQuestFunctionLibrary|AthenaCheckCreativeMode (stale EventBeginPlay _asfort_game_state_athena)))
  (if _cancompletequest
    (Variables|Default|SetCreativeMode false)
    (else
      (Variables|Default|SetCreativeMode true)))
  ;[1 1 1] REPLACE
  (Utilities|FlowControl|Delay 5.0)
  (Networking|SwitchHasAuthority
    (:Remote
      ;[1 1 1] Bind Quest Updates
      (bind (_questisvalid _questcompleted _objectivecompleted) (Default|AthenaCheckQuestandObjectiveCompletion _output (Variables|QuestStuff|GetQuestItem) (Variables|QuestStuff|GetObjBackendName)))
      (if _questisvalid
        (if _objectivecompleted
          _
          (elif (Variables|Default|GetCreativeMode)
            _
            (else
              (Quests|BindEventtoOnQuestsUpdated (Quests|GetQuestManager _output "Invalid") (EventDispatchers|CreateEvent :!event Custom|QuestUpdated))
              (Variables|Default|SetQuestLocallyActive true))))))))

(event Custom|QuestUpdated
  ;[1 1 1] Quest Updated
  (if (Variables|Default|GetQuestLocallyActive)
    (bind (_questisvalid _questcompleted _objectivecompleted) (Default|AthenaCheckQuestandObjectiveCompletion (Utilities|Array|Get(acopy) (Fortnite|GetLocalFortPlayerControllers true)) (Variables|QuestStuff|GetQuestItem) (Variables|QuestStuff|GetObjBackendName)))
    (if (or _questcompleted _objectivecompleted)
      (Variables|Default|SetQuestLocallyActive false)
      (Utilities|FlowControl|Sequence
        (:then_0
          ;[1 1 1] Bind Quest Updates
          (Quests|UnbindEventfromOnQuestsUpdated (Quests|GetQuestManager (Utilities|Array|Get(acopy) (Fortnite|GetLocalFortPlayerControllers true)) "Invalid") (EventDispatchers|CreateEvent :!event Custom|QuestUpdated)))))))

(event OnComponentBeginOverlap(Box)
  (params (OverlappedComponent :type Primitive%20Component%20Object%20Reference) (OtherActor :type Actor%20Object%20Reference) (OtherComp :type Primitive%20Component%20Object%20Reference) (OtherBodyIndex :type Integer) (bFromSweep :type Boolean) (SweepResult :type Hit%20Result%20Structure%20%28by%20ref%29))
  ;[1 1 1] Start timer to update number
  (Networking|SwitchHasAuthority
    (:Authority
      (Utilities|IsValid OtherActor
        (:Is%20Valid
          (bind _returnvalue (Utilities|Time|SetTimerbyEvent (delegate Custom|UpdateSpeed) 0.25 true))
          (Variables|Default|SetSpeedTimer _returnvalue))))))

(event Custom|PlayQuestSuccessFX
  (bind _returnvalue (Fortnite|GetLocalFortPlayerControllers true))
  (bind _output (Utilities|Array|Get(acopy) _returnvalue))
  (bind _returnvalue_2 (Game|Player|GetPlayerPawn _output))
  ;[1 1 1] Quest Success FX
  (Development|PrintString "SUCCESS FX" false false "(R=1.000000,G=0.703857,B=0.040648,A=1.000000)" 5.0)
  (Effects|SpawnFXSystemAttached (Variables|Default|GetQuestSuccessEmitter) (Variables|/Script/Engine.Character|GetMesh _returnvalue_2))
  (Audio|PlaySoundatLocation (Variables|Default|GetQuestSuccessSound) (Transformation|GetActorLocation _returnvalue_2)) ; Quest Success Sound
  (Game|Feedback|ClientPlayForceFeedback (Utilities|Array|Get(acopy) (Fortnite|GetLocalFortPlayerControllers true)) (Variables|Default|GetQuestSuccessForceFeedback) "None" false false false))

(event Custom|UpdateSpeed
  (bind _actorsinrange (Variables|Default|GetActorsInRange))
  (bind _self self)
  (bind _returnvalue (Fortnite|GetClosestActorfromArray _self _actorsinrange))
  ;[1 1 1] Only update if player is in volume (optimisation)
  (if (> (Utilities|Array|Length (Collision|GetOverlappingActors (Variables|Default|GetBox) "/Game/Athena/PlayerPawn_Athena.PlayerPawn_Athena_C" :!class /Script/Engine.PrimitiveComponent)) 0)
    (bind _output_get (Variables|Default|SetActorsInRange (Collision|GetOverlappingActors (Variables|Default|GetBox) 0 :!class /Script/Engine.PrimitiveComponent)))
    ;[1 1 1] Sort through array and remove non moving vehicles
    (for _array_element _output_get
      (if (not (> (Math|Vector|VectorLength (Transformation|GetVelocity _array_element)) 0.0))
        (Utilities|Array|RemoveItem _output_get _array_element)))
    ;[1 1 1] Set speed to MPH, based on Player vs Vehicle.
    (bind _asfort_athena_vehicle (Utilities|Casting|CastToFortAthenaVehicle _returnvalue)
      (:then
        (bind _asfort_vehicle_interface (Utilities|Casting|CastToFortVehicleInterface _asfort_athena_vehicle))
        (Variables|Default|SetMovingVehicleInterface _asfort_vehicle_interface)
        (Variables|Default|SetVehicleTriggersSign true)
        (Variables|Default|SetVehicleSpeed (Math|Float|Absolute(Float) (Vehicle|GetTotalSpeedKmh _asfort_athena_vehicle))))
      (:CastFailed
        (bind _asplayer_pawn_athena (Utilities|Casting|CastToPlayerPawn_Athena _returnvalue))
        (Variables|Default|SetMovingPlayer _asplayer_pawn_athena)
        (Variables|Default|SetVehicleTriggersSign false)
        (Variables|Default|SetVehicleSpeed (* (Math|Vector|VectorLength (Transformation|GetVelocity _asplayer_pawn_athena)) 0.036))))
    (else
      ;[1 1 1] Reset sign
      (Utilities|Time|ClearandInvalidateTimerbyHandle (Variables|Default|GetSpeedTimer))
      (Variables|Default|SetVehicleSpeed 0.0)
      (Utilities|Array|Clear (Variables|Default|GetPlayersInVehicle))
      (Variables|Default|SetMovingPlayer 0))))

(event Collision|EventActorEndOverlap
  (params (OtherActor :type Actor%20Object%20Reference))
  ;[1 1 1] Check for other players in volume to see if we should stop the timer or not.
  (if (not (> (Utilities|Array|Length (Collision|GetOverlappingActors (Variables|Default|GetBox) "/Game/Athena/PlayerPawn_Athena.PlayerPawn_Athena_C" :!class /Script/Engine.PrimitiveComponent)) 0))
    ;[1 1 1] Reset sign
    (Utilities|Time|ClearandInvalidateTimerbyHandle (Variables|Default|GetSpeedTimer))
    (Variables|Default|SetVehicleSpeed 0.0)
    (Utilities|Array|Clear (Variables|Default|GetPlayersInVehicle))
    (Variables|Default|SetMovingPlayer 0)))

(event Custom|GatherPlayersForQuestReward
  (bind _movingvehicleinterface (Variables|Default|GetMovingVehicleInterface))
  (bind _returnvalue (Vehicle|GetDriver _movingvehicleinterface))
  (bind _returnvalue_2 (Vehicle|GetAllPassengers _movingvehicleinterface))
  (bind _movingplayer (Variables|Default|GetMovingPlayer))
  ;[1 0 0] Disabled by flag until needed again for challenge
  ;;[1 1 1] Gather Players for Quest Reward
  (if (Variables|Default|GetCreativeMode)
    _
    (elif (Variables|Default|GetVehicleTriggersSign)
      (Utilities|IsValid _returnvalue
        (:Is%20Valid
          (Utilities|Array|AddUnique (Variables|Default|GetPlayersInVehicle) (Fortnite|GetFortPlayerControllerfromActor _returnvalue)))
        (:Is%20Not%20Valid
          _))
      (if (> (Utilities|Array|Length _returnvalue_2) 0)
        (for _array_element _returnvalue_2
          (Utilities|Array|AddUnique (Variables|Default|GetPlayersInVehicle) (Fortnite|GetFortPlayerControllerfromActor _array_element)))
        (Class|AthenaQuestFunctionLibrary|AthenaGiveQuestUpdatetoPlayers :InQuestItemRef (Variables|QuestStuff|GetQuestItem) :ObjectiveBackendName (Variables|QuestStuff|GetObjBackendName) :PlayerControllersForUpdate (Variables|Default|GetPlayersInVehicle))
        (Development|PrintString "Attempt To Update Quest" false false))
      (else
        (Utilities|IsValid _movingplayer
          (:Is%20Valid
            (Class|AthenaQuestFunctionLibrary|AthenaGiveQuestUpdatetoPlayers :InQuestItemRef (Variables|QuestStuff|GetQuestItem) :ObjectiveBackendName (Variables|QuestStuff|GetObjBackendName) :PlayerControllersForUpdate (Utilities|Array|MakeArray (Fortnite|GetFortPlayerControllerfromActor _movingplayer)))))))))

(event Custom|SpeedLimitExceeded
  (bind _dynamicmat (Variables|Default|GetDynamicMat))
  ;[1 1 1] Play sound and flash if speed limit exceeded
  (Audio|PlaySoundatLocation (Variables|Default|GetBreakSpeedLimitSound) (Transformation|GetActorLocation))
  (Rendering|Material|SetScalarParameterValue (Variables|Default|GetDynamicMat) "Speed" (Variables|Default|GetVehicleSpeed) :!class /Script/Engine.MaterialInstanceDynamic)
  (Variables|Default|SetBlinkNum 0)
  (Rendering|Material|SetScalarParameterValue _dynamicmat "DisplayOn" 1.0 :!class /Script/Engine.MaterialInstanceDynamic :!label _label)
  (Utilities|FlowControl|Delay 0.15)
  (Rendering|Material|SetScalarParameterValue _dynamicmat "DisplayOn" 0.0 :!class /Script/Engine.MaterialInstanceDynamic)
  (Utilities|FlowControl|Delay 0.15)
  (Math|Integer|IncrementInt (Variables|Default|GetBlinkNum)
    (:%20%20
      (if (>= (Variables|Default|GetBlinkNum) 5)
        (Rendering|Material|SetScalarParameterValue (Variables|Default|GetDynamicMat) "DisplayOn" 1.0 :!class /Script/Engine.MaterialInstanceDynamic)
        (else
          (goto _label))))))

(fn OnRep_VehicleSpeed
  (bind _vehiclespeed (Variables|Default|GetVehicleSpeed))
  (Rendering|Material|SetScalarParameterValue (Variables|Default|GetDynamicMat) "Speed" _vehiclespeed :!class /Script/Engine.MaterialInstanceDynamic)
  (if (> _vehiclespeed (Variables|Default|GetSpeedLimit))
    (Networking|SwitchHasAuthority
      (:Authority
        ;[1 0 0] Pruning unnecessary quest updates - set bool to true if reusing speed limit challenge
        (if (Variables|Default|GetShouldGrantQuestReward)
          (CallFunction|GatherPlayersforQuestReward)))
      (:Remote
        (CallFunction|SpeedLimitExceeded)))))

(fn ConstructionScript
  (|CallParentFunction)
  (bind _returnvalue (Rendering|Material|CreateDynamicMaterialInstance (Variables|Debug|GetStaticMeshComponent) 1 "/Game/Environments/Props/Street/Materials/ElectronicSignChallenge.ElectronicSignChallenge" "None" :!class /Script/Engine.PrimitiveComponent))
  (Variables|Default|SetDynamicMat _returnvalue))
