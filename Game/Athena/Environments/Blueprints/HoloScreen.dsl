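; Hologram screen prop for Athena environments, with a projector mesh configured through editable variables.
; The construction script scales the mesh, applies an optional screen material override, then creates dynamic materials for the screen and projection slots.
; It feeds color, glow, contrast, rotation speed, bend and an optional texture into those materials.
; On death it fades out the looping hologram audio; BeginPlay, Tick and overlap events are empty.
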(event EventBeginPlay)

(event Collision|EventActorBeginOverlap
  (params (OtherActor :type Actor%20Object%20Reference)))

(event EventTick
  (params (DeltaSeconds :type Float%20%28single-precision%29)))

(event EventOnDeathPlayEffects
  (params (Damage :type Float%20%28single-precision%29) (DamageTags :type Gameplay%20Tag%20Container%20Structure%20%28by%20ref%29) (Momentum :type Vector :default "0, 0, 0") (HitInfo :type Hit%20Result%20Structure%20%28by%20ref%29) (InstigatedBy :type Fort%20Pawn%20Object%20Reference) (DamageCauser :type Actor%20Object%20Reference) (EffectContext :type Gameplay%20Effect%20Context%20Handle%20Structure))
  (bind _spot_hologram_holo_screen_default_instanced_loop_cue (Variables|Default|GetSpot_Hologram_Holo_Screen_Default_Instanced_Loop_Cue))
  (Utilities|IsValid _spot_hologram_holo_screen_default_instanced_loop_cue
    (:Is%20Valid
      (if (Utilities|IsValid _spot_hologram_holo_screen_default_instanced_loop_cue :!class /Script/Engine.KismetSystemLibrary)
        (Audio|Components|Audio|FadeOut _spot_hologram_holo_screen_default_instanced_loop_cue 0.3 0.0 "Linear" :!class /Script/Engine.AudioComponent)))))

(fn ConstructionScript
  (bind _sm_holoprojector (Variables|Default|GetSM_HoloProjector))
  (Transformation|SetWorldScale3D (Variables|Default|GetSM_HoloProjector) (* (Variables|Default|GetScale) (Variables|Default|GetScaleUniform)))
  (if (Utilities|IsValid (Variables|Default|GetOverrideMaterial_Screen) :!class /Script/Engine.KismetSystemLibrary)
    (Rendering|Material|SetMaterial (Variables|Default|GetSM_HoloProjector) 0 (Variables|Default|GetOverrideMaterial_Screen)))
  (bind _returnvalue (Rendering|Material|CreateDynamicMaterialInstance _sm_holoprojector 1 0 "None" :!class /Script/Engine.PrimitiveComponent))
  (Variables|Default|SetMID_Projection _returnvalue)
  (bind _returnvalue_2 (Rendering|Material|CreateDynamicMaterialInstance _sm_holoprojector 0 0 "None" :!class /Script/Engine.PrimitiveComponent))
  (Variables|Default|SetMID_Screen _returnvalue_2)
  (Rendering|Material|SetScalarParameterValue (Variables|Default|GetMID_Projection) "Constant Rotation Speed" (Variables|Default|GetRotationSpeed) :!class /Script/Engine.MaterialInstanceDynamic)
  (Rendering|Material|SetScalarParameterValue (Variables|Default|GetMID_Projection) "Rotationamount" (* (Variables|Default|GetBend) -0.01) :!class /Script/Engine.MaterialInstanceDynamic)
  (Rendering|Material|SetVectorParameterValue (Variables|Default|GetMID_Screen) "Color" (Variables|Default|GetColor) :!class /Script/Engine.MaterialInstanceDynamic)
  (Rendering|Material|SetScalarParameterValue (Variables|Default|GetMID_Screen) "Glow" (Variables|Default|GetGlowScreen) :!class /Script/Engine.MaterialInstanceDynamic)
  (Rendering|Material|SetScalarParameterValue (Variables|Default|GetMID_Projection) "Glow" (Variables|Default|GetGlowLaser) :!class /Script/Engine.MaterialInstanceDynamic)
  (Rendering|Material|SetScalarParameterValue (Variables|Default|GetMID_Screen) "Contrast" (Variables|Default|GetScreenContrast) :!class /Script/Engine.MaterialInstanceDynamic)
  (Utilities|IsValid (Variables|Default|GetTexture)
    (:Is%20Valid
      (Rendering|Material|SetTextureParameterValue (Variables|Default|GetMID_Projection) "BaseColor" (Variables|Default|GetTexture)))))
