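; Blueprint for a 16x16 Athena foundation that labels itself with text showing its name and size.
; The construction script copies the Name and Size variables into the TextName and TextSize text render components.
; The file holds no other logic, so the labels are set only at construction.
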(fn ConstructionScript
  (Rendering|Components|TextRender|SetText (Variables|Default|GetTextName) (Variables|Default|GetName))
  (Rendering|Components|TextRender|SetText (Variables|Default|GetTextSize) (Variables|Default|GetSize)))
