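; LF_Athena_Grid_64x64 is a Fortnite Athena environment Blueprint that, as its name suggests, acts as a 64x64 grid foundation.
; On BeginPlay it runs the parent logic, then registers itself as a grid foundation; on EndPlay it runs the parent logic, then unregisters.
; The construction script only calls its parent, and the overlap and tick events are empty.
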(event EventBeginPlay
  (|CallParentFunction)
  (Fortnite|RegisterGridFoundation self))

(event EventEndPlay
  (params (EndPlayReason :type EEndPlayReason%20Enum :default "Destroyed"))
  (|CallParentFunction EndPlayReason)
  (Fortnite|UnregisterGridFoundation self))

(event Collision|EventActorBeginOverlap
  (params (OtherActor :type Actor%20Object%20Reference)))

(event EventTick
  (params (DeltaSeconds :type Float%20%28single-precision%29)))

(fn ConstructionScript
  (|CallParentFunction))
