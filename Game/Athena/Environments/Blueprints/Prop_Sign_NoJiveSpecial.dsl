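; No-jive sign prop that launches pawns caught emoting inside its box volume along negative world Y and upward, with a sound.
; Player pawn overlap begin and end each set a looping server timer for a delayed check, letting emotes start first.
; The check launches emoting player-controlled pawns, pausing the stored timer once no players remain.
; BeginPlay clears PlayerCollision, whose OnRep makes the mesh ignore pawns.
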(event EventBeginPlay
  (Variables|Default|SetPlayerCollision false) ; so players dont collide with the sign
)

(event OnComponentBeginOverlap(Box)
  (params (OverlappedComponent :type Primitive%20Component%20Object%20Reference) (OtherActor :type Actor%20Object%20Reference) (OtherComp :type Primitive%20Component%20Object%20Reference) (OtherBodyIndex :type Integer) (bFromSweep :type Boolean) (SweepResult :type Hit%20Result%20Structure%20%28by%20ref%29))
  (Utilities|Casting|CastToFortPlayerPawnAthena OtherActor)
  (Networking|SwitchHasAuthority
    (:Authority
      (bind _returnvalue (Utilities|Time|SetTimerbyEvent (delegate Custom|Checkoverlaps) 2.5 true))
      (Variables|Default|SetTimerHandle _returnvalue))))

(event OnComponentEndOverlap(Box)
  (params (OverlappedComponent :type Primitive%20Component%20Object%20Reference) (OtherActor :type Actor%20Object%20Reference) (OtherComp :type Primitive%20Component%20Object%20Reference) (OtherBodyIndex :type Integer))
  (Utilities|Casting|CastToFortPlayerPawnAthena OtherActor)
  (Networking|SwitchHasAuthority
    (:Authority
      (bind _returnvalue (Utilities|Time|SetTimerbyEvent (delegate Custom|Checkoverlaps) 2.5 true))
      (Variables|Default|SetTimerHandle _returnvalue))))

(event Custom|Checkoverlaps
  (Utilities|FlowControl|Delay 1.5) ; allow emote to start first, feels better
  (if (>= (Utilities|Array|Length (Collision|GetOverlappingActors (Variables|Default|GetBox) "/Script/FortniteGame.FortPlayerPawnAthena" :!class /Script/Engine.PrimitiveComponent)) 1)
    (for _array_element (Collision|GetOverlappingActors (Variables|Default|GetBox) "/Script/FortniteGame.FortPawn" :!class /Script/Engine.PrimitiveComponent)
      (bind _asfort_pawn (Utilities|Casting|CastToFortPawn _array_element))
      (Variables|Default|SetCurrentPawn _asfort_pawn)
      (bind _asfort_player_controller_athena (Utilities|Casting|CastToFortPlayerControllerAthena (Fortnite|GetControllerfromActor _array_element)))
      (if (Emote|IsPlayingEmote _asfort_player_controller_athena)
        (CallFunction|PlayReplicatedDBSFX)
        (Character|LaunchCharacter (Variables|Default|GetCurrentPawn) (Math|Vector|MakeVector 0.0 (+ (.y (Transformation|GetVelocity (Variables|Default|GetCurrentPawn))) -2e+03) (+ (* (.z (Transformation|GetVelocity (Variables|Default|GetCurrentPawn))) -1.0) 1e+03)) false false)))
    (if (not (>= (Utilities|Array|Length (Collision|GetOverlappingActors (Variables|Default|GetBox) "/Script/FortniteGame.FortPlayerPawnAthena" :!class /Script/Engine.PrimitiveComponent)) 1))
      (Utilities|Time|PauseTimerbyHandle (Variables|Default|GetTimerHandle)))
    (else
      (Utilities|Time|PauseTimerbyHandle (Variables|Default|GetTimerHandle)))))

(event Custom|Play%20Replicated%20DB%20SFX
  (Audio|PlaySoundatLocation (Variables|Default|GetSound) (Transformation|GetActorLocation)))

(fn OnRep_PlayerCollision
  (if (not (Variables|Default|GetPlayerCollision))
    (Collision|SetCollisionResponsetoChannel (Variables|Debug|GetStaticMeshComponent) "ECC_Pawn" "ECR_Ignore")))

(fn ConstructionScript)
