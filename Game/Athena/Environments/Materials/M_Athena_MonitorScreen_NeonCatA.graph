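# Unlit translucent material for an Athena environment monitor screen, named for a neon cat and holding a NeonCat monitor texture sample.
# A separate screen chain layers a two-color radial gradient and icon mask, blends toward media video via IsGadget, then scales by Emissive.
# A Transition parameter drives panned-texture UV distortion, scrolling fade overlays and a sine-based vertical UV offset.
# Material output links are not shown, so which branch reaches the screen is unclear.
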Material /Game/Athena/Environments/Materials/M_Athena_MonitorScreen_NeonCatA
graph {
 properties {
  bScreenSpaceReflections=true
  bUsedWithNanite=true
  blendMode=BLEND_Translucent
  numCustomizedUVs=1
  pixelDepthOffsetMode=PDOM_Legacy
  shadingModel=MSM_Unlit
  usedShadingModels=Unlit
 }
 nodes [
  Constant3Vector() {constant=<rgba 0.0084 0.0091 0.0104 0>}
  Constant()
  n1 = VectorParameter() {parameterName=Col_02, defaultValue=<rgba 0.0722 0.4483 1.8646 1>}
  n2 = TextureCoordinate()
  n3 = ComponentMask(None=@n2) {g=true}
  n4 = MaterialFunctionCall(Float2=@n2) {materialFunction=&BreakOutFloat2Components}
  n5 = AppendVector(A=@n4.G, B=@n4.G)
  n6 = Clamp(None=TextureSample(UVs=Panner(Coordinate=@n5) {speedX=-0.25}) {samplerType=SAMPLERTYPE_LinearColor, texture=/Game/Effects/Fort_Effects/BluGlow/BluGlowAnimation_Normals}.R) {maxDefault=0.125}
  n7 = ScalarParameter() {parameterName=Transition}
  n8 = LinearInterpolate(A=TextureCoordinate(), B=AppendVector(A=ComponentMask(None=@n2) {r=true}, B=Add(A=@n3, B=@n6)), Alpha=@n7)
  n9 = MaterialFunctionCall(
   UVs=@n8
   Texture Scale=AppendVector(
    A=ScalarParameter() {defaultValue=0.5, parameterName=UScale_RadialGad}
    B=ScalarParameter() {defaultValue=0.8, parameterName=VScale_RadialGrad}
   )
  ) {materialFunction=&ScaleUVsByCenter}
  n10 = ScalarParameter() {defaultValue=1.2952, parameterName=Radius, sliderMax=2}
  n11 = ScalarParameter() {defaultValue=0.8571, parameterName=Density, sliderMax=10}
  n12 = AppendVector(A=ScalarParameter() {defaultValue=0.45, parameterName=UScale}, B=ScalarParameter() {defaultValue=0.8, parameterName=VScale})
  n13 = LinearInterpolate(
   A=LinearInterpolate(
    A=VectorParameter() {parameterName=Col_01, defaultValue=<rgba 0.0155 0.0536 0.1094 1>}.RGB
    B=@n1.RGB
    Alpha=MaterialFunctionCall(UVs=@n9.UVs, Radius=@n10, Density=@n11) {materialFunction=&RadialGradientExponential}.RadialGradientExponential
   )
   B=Multiply(A=@n1.RGB) {constB=1.5}
   Alpha=TextureSampleParameter2D(UVs=MaterialFunctionCall(UVs=@n8, Texture Scale=@n12) {materialFunction=&ScaleUVsByCenter}.UVs) {parameterName=Upgrade-IconSet, texture=/Game/UI/Foundation/Textures/Icons/Stats/T-Icon-BuildingHealth-128}.A
  )
  n14 = TextureSampleParameter2D() {parameterName=GadgetVid, samplerSource=SSM_Wrap_WorldGroupSettings, texture=/Game/UI/Foundation/Movie/DefaultMediaTexture}
  n15 = ScalarParameter() {parameterName=IsGadget, sliderMax=1}
  n16 = TextureSample(UVs=Panner(Coordinate=@n8) {speedY=-0.1}) {texture=/Game/Environments/Homebase/Textures/T_RoadFade_M}
  n17 = TextureSample(UVs=Panner(Coordinate=@n8) {speedX=0.05}) {texture=&T_RoadFade_M}
  Multiply(
   A=LinearInterpolate(A=LinearInterpolate(A=@n13, B=@n14.RGB, Alpha=@n15), B=Add(A=@n16.G, B=@n17.R), Alpha=LinearInterpolate(Alpha=@n7) {constB=0.1})
   B=ScalarParameter() {defaultValue=1, parameterName=Emissive, sliderMax=1}
  )
  n18 = Constant()
  n19 = Time()
  n20 = LinearInterpolate(B=Clamp(None=Sine(None=@n19) {period=0.125}) {maxDefault=0.25, minDefault=-0.25}, Alpha=@n7)
  Reroute(None=Add(A=TextureCoordinate(), B=AppendVector(A=@n18, B=@n20)))
  ScalarParameter() {defaultValue=0.35, parameterName=Gloss}
  ScalarParameter() {defaultValue=0.25, parameterName=Spec}
  TextureSample() {samplerType=SAMPLERTYPE_Normal, texture=&FlatNormalBlue}
  TextureSample() {texture=/Game/Athena/Environments/Textures/T_NeonCat_MonitorA}
 ]
}
