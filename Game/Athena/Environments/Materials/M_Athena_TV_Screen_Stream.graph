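# TV screen material for Athena environments that shows the StreamingTelevisionMediaTexture stream on an opaque, Nanite-enabled surface.
# An animated BoatFlake flipbook plus Fresnel glow forms the base emissive layer, blended toward the stream by Broadcast when the UseEmergencyBroadCast switch is on.
# Enabled blends from a near-black off state to that result, with roughness forced to 1.
# Destruction, diffuse coloration and flicker subgraphs are left unconnected.
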Material /Game/Athena/Environments/Materials/M_Athena_TV_Screen_Stream
graph {
 properties {
  bUseMaterialAttributes=true
  bUsedWithNanite=true
  opacityMaskClipValue=0.16
  pixelDepthOffsetMode=PDOM_Legacy
  usedShadingModels=Default Lit
 }
 nodes [
  n1 = Time()
  n2 = Sine(None=Multiply(A=@n1) {constB=0.3})
  Abs(None=@n2)
  n3 = TextureCoordinate() {coordinateIndex=1}
  n4 = TextureCoordinate() {uTiling=4, vTiling=4}
  n5 = TextureSample(UVs=@n4) {samplerType=SAMPLERTYPE_Normal, texture=/Game/Global/Textures/Default/Tileable_Patterns/BrokenTilingTextureConcrete_N}
  Add(
   A=Multiply(
    A=TextureSample(UVs=@n3) {samplerType=SAMPLERTYPE_Normal, texture=/Game/Global/Textures/Default/Masks_and_Gradients/RevealMask_n}.RGB
    B=Constant3Vector() {constant=<rgba 1 1 0 0>}
   )
   B=@n5.RGB
  )
  n6 = TextureSample() {texture=/Game/UI/Foundation/Movie/StreamingTelevisionMediaTexture}
  n7 = MaterialFunctionCall!(A, B, Bounds, Edge Falloff=Constant() {r=70}) {materialFunction=/Engine/Functions/Engine_MaterialFunctions01/Gradient/BoxMask-2D}
  n8 = Multiply(A=Time(), B=Constant() {r=3})
  n9 = Constant() {r=2}
  n10 = Add(
   A=MaterialFunctionCall(
    "Animation  Phase (0-1)"=Frac(None=@n8)
    Number of Rows=@n9
    Number of Columns=@n9
    Texture=TextureObject() {texture=/Game/Global/Textures/Default/PointSplatter/BoatFlake}
    UVs=TextureCoordinate() {uTiling=3, vTiling=3}
   ) {materialFunction=&FlipBook}.Result
   B=Multiply(A=Fresnel() {baseReflectFraction=0, exponent=7}) {constB=0.4}
  )
  n11 = Abs(None=Sine(None=Multiply(A=@n1) {constB=0.666}))
  Add(A=Add(A=Multiply(A=@n6.RGB, B=@n7.Result), B=Multiply(A=@n10) {constB=0.5}), B=Multiply(A=Multiply(A=@n2, B=@n11)) {constB=0.2})
  n12 = VectorParameter() {group=Diffuse, parameterName=DiffuseColoration, defaultValue=<rgba 1 1 1 0>}
  AppendVector(A=@n12.RGB, B=@n12.RGB)
  n13 = TextureSample(UVs=TextureCoordinate() {uTiling=4, vTiling=4}) {samplerType=SAMPLERTYPE_LinearColor, texture=/Game/Global/Textures/Default/Noise/T_Noise01}
  n14 = TextureSample(UVs=TextureCoordinate() {coordinateIndex=1}) {samplerType=SAMPLERTYPE_LinearColor, texture=/Game/Global/Textures/Default/Masks_and_Gradients/RevealMask}
  n15 = Multiply(A=ConstantBiasScale(None=@n13.R) {bias=-0.1, scale=1}, B=@n14.R)
  n16 = ComponentMask(None=@n5.RGB) {b=true}
  n17 = Multiply(A=Multiply(A=@n14.R, B=OneMinus(None=@n16)), B=Constant() {r=2})
  Clamp(None=Add(A=Add(A=@n15, B=@n17), B=@n14.R))
  n18 = Multiply(A=@n13.R, B=Constant() {r=0.1})
  n19 = Clamp(None=ScalarParameter() {group=Code_Destruction, parameterName=DestructionAmount})
  n20 = Constant() {r=2}
  n21 = OneMinus(None=MaterialFunctionCall!(Lerp A, Lerp B, Height Texture=Add(A=@n14.R, B=@n18), Falloff Contrast=Constant() {r=2}, Transition Phase=Divide(A=@n19, B=@n20)) {materialFunction=/Game/Packages/Fortnight_MaterialFunctions/Temp/HeightLerp}.Lerp Alpha)
  Multiply(A=@n21, B=@n21)
  ScalarParameter() {defaultValue=0.65, group=Diffuse, parameterName=DiffuseColorationThreshold}
  n22 = Constant3Vector() {constant=<rgba 0.0104 0.0104 0.0104 0>}
  n23 = MakeMaterialAttributes!(
   BaseColor=Constant()
   Metallic
   Specular
   Roughness
   Anisotropy
   EmissiveColor=@n10
   Opacity
   OpacityMask
   Normal
   Tangent
   WorldPositionOffset
   SubsurfaceColor
   ClearCoat
   ClearCoatRoughness
   AmbientOcclusion
   Refraction
   CustomizedUVs_0
   CustomizedUVs_1
   CustomizedUVs_2
   CustomizedUVs_3
   CustomizedUVs_4
   CustomizedUVs_5
   CustomizedUVs_6
   CustomizedUVs_7
   PixelDepthOffset
   ShadingModel
   Displacement
  )
  n24 = MaterialFunctionCall!(
   Base Material=@n23
   Top Material=MakeMaterialAttributes(BaseColor=@n6.RGB, EmissiveColor=@n6.RGB)
   Alpha=ScalarParameter() {defaultValue=1, parameterName=Broadcast}
  ) {materialFunction=/Engine/Functions/MaterialLayerFunctions/MatLayerBlend_Standard}
  SetMaterialAttributes!(
   MaterialAttributes=MaterialFunctionCall(
    Base Material=MakeMaterialAttributes(BaseColor=@n22, EmissiveColor=@n22)
    Top Material=StaticSwitchParameter(True=@n24.Blended Material, False=@n23) {defaultValue=true, parameterName=UseEmergencyBroadCast}
    Alpha=ScalarParameter() {defaultValue=1, parameterName=Enabled}
   ) {materialFunction=&MatLayerBlend_Standard}.Blended Material
   Roughness=Constant() {r=1}
  ) {attributeSetTypes [#1]}
  StaticBoolParameter() {defaultValue=true, group=Diffuse, parameterName=UseProceduralDiffuseColorationMask}
  n25 = TextureSampleParameter2D() {group=Diffuse, parameterName=DiffuseColorationMask, texture=/Game/Global/Textures/Default/Masks_and_Gradients/16x}
  StaticComponentMaskParameter(None=AppendVector(A=@n25.RGB, B=@n25.RGB)) {defaultR=true, group=Diffuse, parameterName=DiffuseColorationMaskComponentMask}
  TextureSample(UVs=@n4) {texture=/Game/Global/Textures/Default/Tileable_Patterns/BrokenTilingTextureConcrete_D}
  TextureSample() {texture=/Game/Athena/Environments/Textures/Screen_WS}
 ]
}
