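# Material function for the Llamatron greenscreen that turns an input texture, previewed with the streaming television media texture, into a grid of RGB value readouts.
# It samples one color per cell of a 12x12 grid and scales it to 0-255.
# Each cell stacks three numbers, extracting digits with power-of-ten math and drawing them from a number-sheet atlas.
# The masked digits pack into matching red, green and blue channels, blended heavily over the pixelated color.
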MaterialFunction /Game/Athena/Environments/Materials/MF_LlamatronGreenscreen_ColorNumbers
graph {nodes [
  Add(B=Constant2Vector() {g=0.33})
  n1 = Multiply(A=TextureCoordinate()) {constB=12}
  n2 = TextureSample(
   UVs=Divide(A=Floor(None=@n1)) {constB=12}
   Tex=FunctionInput(Preview=TextureObject() {texture=/Game/UI/Foundation/Movie/StreamingTelevisionMediaTexture}) {bUsePreviewValueAsDefault=true, inputType=FunctionInput_Texture2D}
  )
  n3 = Frac(None=@n1)
  n4 = MaterialFunctionCall(UVs=Add(A=@n3, B=Constant2Vector() {g=0.25}), Texture Scale=Constant() {r=0.2}) {materialFunction=&ScaleUVsByCenter}
  n5 = Multiply(A=@n4.UVs, B=Constant2Vector() {g=0.1, r=1}) {constA=2, constB=0.1}
  n6 = Multiply(A=OneMinus(None=ComponentMask(None=@n5) {r=true}), B=Constant() {r=3}) {constB=10}
  n7 = Frac(None=@n6)
  n8 = Multiply(A=@n2.RGB) {constB=255}
  n9 = ComponentMask(None=@n8) {r=true}
  n10 = Max(A=@n6) {constB=0}
  n11 = Power(Base=Constant() {r=10}, Exp=Add(A=Ceil(None=@n10))) {desc="10-100,000,000"}
  n12 = Multiply(A=Frac(None=Divide(A=@n9, B=@n11))) {constB=100}
  n13 = Add(A=ComponentMask(None=@n5) {g=true}, B=Divide(A=Floor(None=@n12)) {constB=10})
  n14 = Multiply(
   A=@n4.0-1 mask
   B=TextureSample!(UVs=AppendVector(A=@n7, B=@n13), Tex, Level, Apply View MipBias) {mipValueMode=TMVM_MipLevel, samplerType=SAMPLERTYPE_Masks, texture=/Game/Athena/Items/Gameplay/VendingMachine/FX/T_NumberSheet}.R
  )
  n15 = MaterialFunctionCall(UVs=@n3, Texture Scale=Constant() {r=0.2}) {materialFunction=&ScaleUVsByCenter}
  n16 = Multiply(A=@n15.UVs, B=Constant2Vector() {g=0.1, r=1}) {constA=2, constB=0.1}
  n17 = Multiply(A=OneMinus(None=ComponentMask(None=@n16) {r=true}), B=Constant() {r=3}) {constB=10}
  n18 = Frac(None=@n17)
  n19 = ComponentMask(None=@n8) {g=true}
  n20 = Max(A=@n17) {constB=0}
  n21 = Power(Base=Constant() {r=10}, Exp=Add(A=Ceil(None=@n20))) {desc="10-100,000,000"}
  n22 = Multiply(A=Frac(None=Divide(A=@n19, B=@n21))) {constB=100}
  n23 = Add(A=ComponentMask(None=@n16) {g=true}, B=Divide(A=Floor(None=@n22)) {constB=10})
  n24 = Multiply(
   A=@n15.0-1 mask
   B=TextureSample(UVs=AppendVector(A=@n18, B=@n23)) {mipValueMode=TMVM_MipLevel, samplerType=SAMPLERTYPE_Masks, texture=&T_NumberSheet}.R
  )
  n25 = MaterialFunctionCall(UVs=Add(A=@n3, B=Constant2Vector() {g=-0.25}), Texture Scale=Constant() {r=0.2}) {materialFunction=&ScaleUVsByCenter}
  n26 = Multiply(A=@n25.UVs, B=Constant2Vector() {g=0.1, r=1}) {constA=2, constB=0.1}
  n27 = Multiply(A=OneMinus(None=ComponentMask(None=@n26) {r=true}), B=Constant() {r=3}) {constB=10}
  n28 = Frac(None=@n27)
  n29 = ComponentMask(None=@n8) {b=true}
  n30 = Max(A=@n27) {constB=0}
  n31 = Power(Base=Constant() {r=10}, Exp=Add(A=Ceil(None=@n30))) {desc="10-100,000,000"}
  n32 = Multiply(A=Frac(None=Divide(A=@n29, B=@n31))) {constB=100}
  n33 = Add(A=ComponentMask(None=@n26) {g=true}, B=Divide(A=Floor(None=@n32)) {constB=10})
  n34 = Multiply(
   A=@n25.0-1 mask
   B=TextureSample(UVs=AppendVector(A=@n28, B=@n33)) {mipValueMode=TMVM_MipLevel, samplerType=SAMPLERTYPE_Masks, texture=&T_NumberSheet}.R
  )
  FunctionOutput(None=LinearInterpolate(
    A=@n2.RGB
    B=MaterialFunctionCall(X=@n14, Y=@n24, Z=@n34) {materialFunction=&MakeFloat3}.Result
    Alpha=Constant() {r=0.95}
   ))
 ]}
