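# Material function that chroma-keys a Llamatron greenscreen texture, outputting Emissive, Opacity and a despill DEBUG mask.
# A custom HLSL cross-shaped blur, converted to sRGB, feeds Chroma_Key_Alpha, softening only the alpha and despill masks.
# Emissive takes the texture through MF_invertFNtonemap, subtracts key-color spill weighted by despill alpha and amount, then saturates.
# The blur divides by four more taps than it samples, darkening its result when iterations are positive.
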MaterialFunction /Game/Athena/Environments/Materials/MF_LlamatronGreenscreen_V2
graph {nodes [
  n1 = MaterialFunctionCall!(Linear=FunctionInput(Preview=Constant3Vector() {constant=<rgba 0.107 0.6308 0 1>}) {bUsePreviewValueAsDefault=true, inputName=ChormaKey, sortPriority=2}) {desc=Needed if texture is brought in with sRGB linearization, materialFunction=/Game/Movies/InversePostProcess/LinearTosRGB3}
  MaterialExpressionFunctionInput_0 = FunctionInput() {inputName=LLamatron Texture, inputType=FunctionInput_Texture2D, sortPriority=1}
  n2 = MaterialFunctionCall!(UVs, Texture=@MaterialExpressionFunctionInput_0) {materialFunction=/Game/Movies/InversePostProcess/MF_invertFNtonemap}
  MaterialExpressionTextureCoordinate_0 = TextureCoordinate()
  MaterialExpressionFunctionInput_1 = FunctionInput(Preview=Constant() {r=0.002}) {inputName=Blur Amount, inputType=FunctionInput_Scalar, sortPriority=30}
  MaterialExpressionFunctionInput_9 = FunctionInput(Preview=Constant()) {inputName=Blur Iterations, inputType=FunctionInput_Scalar, sortPriority=31}
  n3 = Custom!(
   Tex=@MaterialExpressionFunctionInput_0
   UV=@MaterialExpressionTextureCoordinate_0
   Offset=@MaterialExpressionFunctionInput_1
   Iterations=@MaterialExpressionFunctionInput_9
  ) {
   code="float3 color = float3(0,0,0);\r\n\r\nint kernelN = 9;\r\n\r\nint samples = int(Iterations);\r\n\r\ncolor = Texture2DSample(Tex, TexSampler, UV);\r\n\r\nfor(int i = 1; i < samples ; i++)\r\n{\r\n\t\tcolor += Texture2DSample(Tex, TexSampler, UV + Offset * float2(-i,0));\r\n\t\tcolor += Texture2DSample(Tex, TexSampler, UV + Offset * float2(i,0));\r\n\t\tcolor += Texture2DSample(Tex, TexSampler, UV + Offset * float2(0,i));\r\n\t\tcolor += Texture2DSample(Tex, TexSampler, UV + Offset * float2(0,-i));\r\n\r\n}\r\n\r\ncolor = color / (1 + 4 * samples);\r\n\r\n\r\n\r\nreturn color;\r\n"
   inputs [
    {inputName=Tex, input {expression=&self:MaterialExpressionFunctionInput_0}}
    {inputName=UV, input {expression=&self:MaterialExpressionTextureCoordinate_0}}
    {inputName=Offset, input {expression=&self:MaterialExpressionFunctionInput_1}}
    {inputName=Iterations, input {expression=&self:MaterialExpressionFunctionInput_9}}
   ]
  }
  n4 = MaterialFunctionCall!(
   LumaMask
   Image Color=MaterialFunctionCall(Linear=@n3) {desc=Needed if texture is brought in with sRGB linearization, materialFunction=&LinearTosRGB3}.sRGB
   Chroma Color=@n1.sRGB
   Alpha Cutoff Min=FunctionInput() {inputName=Alpha Cutoff Min, inputType=FunctionInput_Scalar, sortPriority=10}
   Alpha Cutoff Max=FunctionInput() {inputName=Alpha Cutoff Max, inputType=FunctionInput_Scalar, sortPriority=11}
   Alpha Exponent=FunctionInput() {inputName=Alpha Exponent, inputType=FunctionInput_Scalar, sortPriority=12}
   Despill Cuttoff Max=FunctionInput() {inputName="Despill Cuttoff Max (S)", inputType=FunctionInput_Scalar, sortPriority=13}
   Despill Exponent=FunctionInput() {inputName=Despill Exponent, inputType=FunctionInput_Scalar, sortPriority=14}
  ) {materialFunction=/Engine/Functions/Engine_MaterialFunctions01/Opacity/Chroma_Key_Alpha}
  n5 = Multiply(
   A=Multiply(A=@n2.Result, B=Normalize(VectorInput=@n1.sRGB))
   B=Multiply(A=@n4.Despill Alpha, B=FunctionInput() {inputName=Despill Amount, inputType=FunctionInput_Scalar, sortPriority=15})
  )
  FunctionOutput(None=Saturate(None=Subtract(A=@n2.Result, B=@n5))) {outputName=Emissive}
  FunctionOutput(None=@n4.Alpha) {outputName=Opacity, sortPriority=1}
  FunctionOutput(None=@n4.Despill Alpha) {outputName=DEBUG, sortPriority=1}
 ]}
