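# Translucent, unlit TV screen material for Creative that displays a streaming media texture, with loading and fallback looks.
# The stream is inverse-tonemapped or optionally chroma-keyed with despill and transparency, with an RGB-value debug view.
# Broadcast blends it over an animated flipbook with Fresnel glow, and a static switch can leave only that flipbook.
# Enabled blends this over a near-black base where Loading reveals a procedural spinning ring built from radial math.
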Material /Game/Athena/Environments/Materials/M_Athena_TV_Screen_Stream_Creative
graph {
 properties {
  bUseMaterialAttributes=true
  bUsedWithNanite=true
  blendMode=BLEND_Translucent
  opacityMaskClipValue=0.16
  pixelDepthOffsetMode=PDOM_Legacy
  shadingModel=MSM_Unlit
  usedShadingModels=Unlit
 }
 nodes [
  n1 = ScalarParameter() {parameterName=Loading}
  n2 = Constant3Vector() {constant=<rgba 0.005 0.005 0.005 0>}
  n3 = ScalarParameter() {defaultValue=0.4, parameterName=MinFill}
  n4 = ScalarParameter() {defaultValue=0.9, parameterName=MaxFill}
  n5 = ScalarParameter() {defaultValue=1.8, parameterName=Speed}
  n6 = MaterialFunctionCall(Speed=Multiply(A=@n5) {constB=0.6}) {materialFunction=&TimeWithSpeedVariable}
  n7 = Add(
   A=@n3
   B=Multiply(
    A=Subtract(A=@n4, B=@n3)
    B=MaterialFunctionCall(Value=@n6.Frac Time) {materialFunction=&LinearSine}."Rounded Linear Sine "
   )
  )
  n8 = MaterialFunctionCall(UVs=TextureCoordinate(), Texture Scale=Constant2Vector() {desc="Loading Icon Scale, squashed to match a 16:9 res", g=0.25, r=0.1406}) {materialFunction=&ScaleUVsByCenter}
  n9 = MaterialFunctionCall(Speed=Multiply(A=@n5)) {materialFunction=&TimeWithSpeedVariable}
  n10 = Multiply(
   A=MaterialFunctionCall(Value=@n9.Frac Time) {materialFunction=&LinearSine}."Rounded Linear Sine "
   B=ScalarParameter() {defaultValue=0.05, parameterName=MaxSpeed}
  )
  n11 = MaterialFunctionCall(UVs=@n8.UVs, "Rotation Angle (0-1)"=Multiply(A=Add(A=@n9.Frac Time, B=@n10)) {constB=-1}) {materialFunction=&CustomRotator}
  n12 = MaterialFunctionCall(Vector or UVs=Multiply(A=Subtract(A=@n11.Rotated Values) {constB=0.5}, B=Constant2Vector() {g=2, r=2}) {constB=2}) {materialFunction=&VectorToRadialValue}
  n13 = Multiply(A=MaterialFunctionCall(Alpha=Add(A=@n7, B=@n12.Vector Converted to Angle), Min=Constant() {r=1}, Max=Constant() {r=0.95}) {materialFunction=&SmoothStep}.Result) {constB=50}
  n14 = Constant() {r=0.5}
  n15 = Clamp(None=Multiply(A=MaterialFunctionCall!(UVs=@n8.UVs, CenterPosition, Radius=@n14) {materialFunction=/Engine/Functions/Engine_MaterialFunctions01/Gradient/SphereGradient-2D}.Result 0-1) {constB=2})
  n16 = ScalarParameter() {defaultValue=0.3, parameterName=MinInnerRadius}
  n17 = Subtract(A=ScalarParameter() {defaultValue=0.6, parameterName=MaxInnerRadius}, B=@n16)
  n18 = Multiply(A=Add(A=Multiply(B=@n17) {constA=1}, B=@n16)) {constB=0.5}
  n19 = Clamp(None=Multiply(A=MaterialFunctionCall(UVs=@n8.UVs, Radius=@n18) {materialFunction=&SphereGradient-2D}.Result 0-1) {constB=2})
  n20 = MaterialFunctionCall(Alpha=Multiply(A=Clamp(None=@n13), B=Subtract(A=@n15, B=@n19)), Min=Constant(), Max=Constant() {r=0.2}) {materialFunction=&SmoothStep}
  n21 = If(A=@n1, "A > B"=@n20.Result, "A == B"=@n2, "A < B"=@n2)
  n22 = Constant()
  n23 = If(A=@n1, "A > B"=@n20.Result, "A == B"=@n22, "A < B"=@n22)
  n24 = Frac(None=Multiply(A=Time(), B=Constant() {r=3}))
  n25 = Constant() {r=2}
  n26 = TextureObject() {texture=/Game/Global/Textures/Default/PointSplatter/BoatFlake}
  n27 = TextureCoordinate() {uTiling=3, vTiling=3}
  n28 = Fresnel() {baseReflectFraction=0, exponent=7}
  n29 = MakeMaterialAttributes!(
   BaseColor=Constant()
   Metallic
   Specular
   Roughness
   Anisotropy
   EmissiveColor=Add(
    A=MaterialFunctionCall("Animation  Phase (0-1)"=@n24, Number of Rows=@n25, Number of Columns=@n25, Texture=@n26, UVs=@n27) {materialFunction=&FlipBook}.Result
    B=Multiply(A=@n28) {constB=0.4}
   )
   Opacity
   OpacityMask
   Normal
   Tangent
   WorldPositionOffset
   SubsurfaceColor
   ClearCoat
   ClearCoatRoughness
   AmbientOcclusion
   Refraction
   CustomizedUVs_0
   CustomizedUVs_1
   CustomizedUVs_2
   CustomizedUVs_3
   CustomizedUVs_4
   CustomizedUVs_5
   CustomizedUVs_6
   CustomizedUVs_7
   PixelDepthOffset
   ShadingModel
   Displacement
  )
  n30 = TextureObject() {texture=/Game/UI/Foundation/Movie/StreamingTelevisionMediaTexture}
  n31 = Divide(A=LinearInterpolate(
    A=ScalarParameter() {defaultValue=5, group=Greenscreen, parameterName="Alpha Cutoff Min (%) A"}
    B=ScalarParameter() {defaultValue=20, group=Greenscreen, parameterName="Alpha Cutoff Min (%) B"}
    Alpha=ScalarParameter() {defaultValue=0.5, parameterName=Cutoff Adjustment, sliderMax=1}
   )) {constB=100}
  n32 = ScalarParameter() {defaultValue=5, group=Greenscreen, parameterName="Alpha Delta Cutoff (%)"}
  n33 = ScalarParameter() {defaultValue=5, group=Greenscreen, parameterName="Despill Cuttoff Max (%)"}
  n34 = MaterialFunctionCall!(
   LLamatron Texture=@n30
   ChormaKey=VectorParameter() {group=Greenscreen, parameterName=Chroma Key, defaultValue=<rgba 0.1329 0.7157 0.0056 1>}.RGB
   Alpha Cutoff Min=@n31
   Alpha Cutoff Max=Add(A=@n31, B=Divide(A=@n32) {constB=100})
   Alpha Exponent=ScalarParameter() {defaultValue=0.5, parameterName=Alpha Exponent}
   "Despill Cuttoff Max (S)"=Add(A=@n31, B=Divide(A=@n33) {constB=100})
   Despill Exponent=ScalarParameter() {defaultValue=1, parameterName=Despill Exponent}
   Despill Amount=ScalarParameter() {defaultValue=1, group=Greenscreen, parameterName=Despill amount}
   Blur Amount=Divide(A=ScalarParameter() {defaultValue=2, parameterName=Blur Amount, sliderMax=10}) {constB=1000}
   Blur Iterations=ScalarParameter() {defaultValue=2, parameterName=Blur Iterations, sliderMax=3}
  ) {materialFunction=&/MF_LlamatronGreenscreen_V2}
  n35 = ScalarParameter() {parameterName=Greenscreen}
  n36 = MaterialFunctionCall!(UVs, Texture=@n30) {materialFunction=/Game/Movies/InversePostProcess/MF_invertFNtonemap}
  n37 = ScalarParameter() {parameterName=View RGB values}
  n38 = LinearInterpolate(
   A=LinearInterpolate(A=@n36.Result, B=@n34.Emissive, Alpha=@n35)
   B=MaterialFunctionCall!(In=@n30) {materialFunction=&/MF_LlamatronGreenscreen_ColorNumbers}.Result
   Alpha=@n37
  )
  n39 = StaticBoolParameter() {defaultValue=true, parameterName=Enable Greenscreen}
  n40 = StaticSwitch(
   True=LinearInterpolate(A=LinearInterpolate(B=@n34.Opacity, Alpha=@n35) {constA=1}, B=Constant() {r=1}, Alpha=@n37)
   False=Constant() {r=1}
   Value=@n39
  )
  n41 = MaterialFunctionCall!(
   Base Material=@n29
   Top Material=MakeMaterialAttributes(EmissiveColor=StaticSwitch(True=@n38, False=@n36.Result, Value=@n39), Opacity=@n40, OpacityMask=@n40)
   Alpha=ScalarParameter() {defaultValue=1, parameterName=Broadcast}
  ) {materialFunction=/Engine/Functions/MaterialLayerFunctions/MatLayerBlend_Standard}
  SetMaterialAttributes!(
   MaterialAttributes=MaterialFunctionCall(
    Base Material=MakeMaterialAttributes(BaseColor=@n21, EmissiveColor=@n23)
    Top Material=StaticSwitchParameter(True=@n41.Blended Material, False=@n29) {defaultValue=true, parameterName=UseEmergencyBroadCast}
    Alpha=ScalarParameter() {defaultValue=1, parameterName=Enabled}
   ) {materialFunction=&MatLayerBlend_Standard}.Blended Material
   Roughness=Constant() {r=1}
  ) {attributeSetTypes [#1]}
 ]
}
