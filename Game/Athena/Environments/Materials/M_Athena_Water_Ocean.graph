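# Lit ocean water material for the Athena map, using world-space normals and world-position UVs throughout.
# A world-mapped terrain mask, scaled by vertex alpha, blends deep to shallow color, with high quality adding mask-limited foam from two subtracted panning textures.
# Panning normal layers flatten with camera distance and Fresnel, with a cheaper blend for low quality and mobile.
# A ground twist function supplies distortion only at high quality.
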Material /Game/Athena/Environments/Materials/M_Athena_Water_Ocean
graph {
 properties {bTangentSpaceNormal=false, pixelDepthOffsetMode=PDOM_Legacy, usedShadingModels=Default Lit}
 nodes [
  Constant()
  LinearInterpolate(A=Constant() {r=1}, B=Constant() {r=0.85})
  MaterialFunctionCall() {materialFunction=&FlattenNormal}
  n1 = ComponentMask(None=Divide(A=WorldPosition().XYZ, B=Constant() {r=18286.0156})) {g=true, r=true}
  n2 = Multiply(A=Time(), B=Constant() {r=0.6})
  n3 = Panner(Coordinate=Multiply(A=Constant() {r=0.2557}, B=@n1), Time=@n2) {bFractionalPart=true, speedX=0.01, speedY=-0.005}
  Rotator(Coordinate=@n3, Time=Constant() {r=0.8658})
  n4 = ComponentMask(None=WorldPosition().XYZ) {g=true, r=true}
  n5 = Constant2Vector() {g=-153000, r=-153000}
  n6 = TextureSample(UVs=Divide(A=Subtract(A=@n4, B=@n5), B=Constant() {r=306200})) {
   samplerSource=SSM_Clamp_WorldGroupSettings
   samplerType=SAMPLERTYPE_LinearColor
   texture=/Game/Athena/Prototype/Terrain/Textures/T_Athena_Terrain_Mask
  }
  n7 = Constant() {r=0.2}
  n8 = LinearInterpolate(
   A=VectorParameter() {parameterName=deepwater, defaultValue=<rgba 0.0049 0.0249 0.05 1>}.RGB
   B=VectorParameter() {parameterName=shallowwater, defaultValue=<rgba 0.0469 0.5 0.4635 1>}.RGB
   Alpha=Multiply(A=Clamp(None=@n6.A, Max=@n7) {clampMode=CMODE_ClampMin, maxDefault=0.5}, B=VertexColor().A)
  )
  n9 = ComponentMask(None=Divide(A=WorldPosition().XYZ, B=Constant() {r=6202.6328})) {g=true, r=true}
  n10 = Time()
  n11 = Multiply(A=@n9) {constB=1.5}
  n12 = Subtract(
   A=TextureSample(UVs=Panner(Coordinate=@n9, Time=@n10) {bFractionalPart=true, speedX=-0.03, speedY=0.03}) {texture=/Game/Global/Textures/Default/Tileable_Patterns/Foam}.G
   B=TextureSample(UVs=Panner(Coordinate=@n11, Time=@n10) {bFractionalPart=true, speedX=0.02, speedY=-0.1}) {texture=&Foam}.G
  )
  n13 = Multiply(A=@n6.A, B=Constant() {r=3})
  n14 = Multiply(A=Multiply(A=Constant3Vector() {constant=<rgba 1.5469 2.5102 3 0>}, B=Clamp(None=@n12)), B=Clamp(None=Subtract(A=@n13) {constB=0.6}))
  ShadingPathSwitch(Default=QualitySwitch(Default=@n8, Low=@n8, High=Add(A=@n8, B=@n14), Medium=@n8), Mobile=@n8) {desc=Migrated from FeatureLevelSwitch}
  n15 = Multiply(A=Constant() {r=0.119}, B=@n1)
  n16 = MaterialFunctionCall(
   BaseNormal=TextureSample(UVs=Panner(Coordinate=@n15, Time=@n2) {bFractionalPart=true, speedX=-0.015, speedY=0.005}) {samplerType=SAMPLERTYPE_Normal, texture=/Game/Global/Textures/Default/Tileable_Patterns/T_Water2_N}.RGB
   AdditionalNormal=TextureSample(UVs=@n3) {samplerType=SAMPLERTYPE_Normal, texture=/Game/Effects/Fort_Effects/Textures/WaterFoam0001_19_N}.RGB
  ) {materialFunction=&BlendAngleCorrectedNormals}
  n17 = MaterialFunctionCall(Fade Length=Constant() {r=185195.2031}, Fade Offset=Constant()) {materialFunction=&CameraDepthFade}
  n18 = LinearInterpolate(A=Constant() {r=1}, B=Constant() {r=0.85}, Alpha=@n17.Result)
  n19 = Time()
  n20 = Multiply(A=@n1, B=Constant() {r=1})
  n21 = Add(
   A=TextureSample(UVs=Panner(Coordinate=@n1, Time=@n19) {bFractionalPart=true, speedX=0.009, speedY=-0.009}) {samplerType=SAMPLERTYPE_Normal, texture=&T_Water2_N}.RGB
   B=TextureSample(UVs=Panner(Coordinate=@n20, Time=@n19) {bFractionalPart=true, speedX=-0.003, speedY=0.006}) {samplerType=SAMPLERTYPE_Normal, texture=&WaterFoam0001_19_N}.RGB
  )
  n22 = LinearInterpolate(A=Constant() {r=0.2}, B=Constant() {r=1}, Alpha=@n17.Result)
  n23 = Fresnel(ExponentIn=Constant() {r=4}, Normal=VertexNormalWS()) {baseReflectFraction=0}
  n24 = MaterialFunctionCall(
   Normal=Add(
    A=MaterialFunctionCall(Normal=@n16.Result, Flatness=@n18) {materialFunction=&FlattenNormal}.Result
    B=MaterialFunctionCall(Normal=@n21, Flatness=@n22) {materialFunction=&FlattenNormal}.Result
   )
   Flatness=@n23
  ) {materialFunction=&FlattenNormal}
  n25 = MaterialFunctionCall(Normal=@n16.Result, Flatness=LinearInterpolate(A=Constant() {r=0.85}, B=Constant() {r=1}, Alpha=@n23)) {materialFunction=&FlattenNormal}
  ShadingPathSwitch(Default=QualitySwitch(Default=@n24.Result, Low=@n25.Result, High=@n24.Result, Medium=@n24.Result), Mobile=@n25.Result) {desc=Migrated from FeatureLevelSwitch}
  n26 = Constant()
  n27 = ScalarParameter() {defaultValue=0.0519, parameterName=Twist_bias}
  n28 = ScalarParameter() {defaultValue=9.7199, parameterName=Twist_Mod}
  n29 = ScalarParameter() {defaultValue=0.5, parameterName=Twist_SineOffset}
  n30 = ScalarParameter() {defaultValue=1, parameterName=Twist_SineSpeed}
  n31 = Multiply(A=@n1, B=ScalarParameter() {defaultValue=0.175, parameterName=Twist_Size})
  n32 = ScalarParameter() {defaultValue=0.2385, parameterName=Twist_RotationMod}
  n33 = ScalarParameter() {defaultValue=1, parameterName=Twist_Time}
  n34 = AppendVector(
   A=ScalarParameter() {defaultValue=-0.6433, group=Twist, parameterName=GridNoiseTileRep}
   B=ScalarParameter() {defaultValue=136.6535, group=Twist, parameterName=GridNoiseTileRep2}
  )
  n35 = ScalarParameter() {defaultValue=1.6632, parameterName=Twist_Falloff}
  ShadingPathSwitch(
   Default=QualitySwitch(
    Default=@n26
    Low=@n26
    High=MaterialFunctionCall!(
     Result Bias=@n27
     Result Modulation=@n28
     SineAnimOffset=@n29
     SineAnimSpeed=@n30
     UVS * size=@n31
     Rotation amount modulator=@n32
     Time=@n33
     Flow Direction
     Grid Noise Tile Rep=@n34
     Twist Falloff=@n35
    ) {materialFunction=/Game/Environments/Sets/Arid/Materials/Fortnite_GroundTwistFunction}.Distortion
    Medium=@n26
   )
   Mobile=Constant()
  ) {desc=Migrated from FeatureLevelSwitch}
  Time()
 ]
}
