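# Translucent, two-sided unlit hologram material, defaulting to Hunters Haven pancake sign textures, that spins the mesh around UV-encoded pivots with corrected normals.
# Emissive adds Fresnel rim, tinted texture and a vertically sweeping sphere-mask pulse that also displaces vertices along normals.
# Opacity combines screen-space scanlines, camera-distance fade and optional particle alpha.
# A custom depth comparison hides rear layers; low-end feature levels drop the pulse and use distance-based backface fading.
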Material /Game/Athena/Environments/Materials/M_Hologram_Neo_Butter
graph {
 properties {
  allowTranslucentCustomDepthWrites=true
  bUsedWithMeshDeformer=true
  bUsedWithMorphTargets=true
  bUsedWithNiagaraMeshParticles=true
  bUsedWithSkeletalMesh=true
  blendMode=BLEND_Translucent
  pixelDepthOffsetMode=PDOM_Legacy
  shadingModel=MSM_Unlit
  twoSided=true
  usedShadingModels=Unlit
 }
 nodes [
  n1 = VectorParameter() {parameterName=FresnelColor, defaultValue=<rgba 0.25 0.1191 0.2226 1>}
  n2 = Normalize(VectorInput=VectorParameter() {parameterName=RotationAxis, defaultValue=<rgba 0 0 1 1>}.RGB)
  n3 = Frac(None=Multiply(A=Time(), B=ScalarParameter() {defaultValue=0.0857, parameterName=SpeedRotation}))
  n4 = VertexInterpolator(VS=MaterialFunctionCall(Rotation Axis=@n2, Rotation Angle=@n3) {materialFunction=&FixRotateAboutAxisNormals}.New Tangent Space Vertex Normal)
  n5 = TextureSampleParameter2D() {
   parameterName=Normal
   samplerType=SAMPLERTYPE_Normal
   texture=/Game/Environments/Apollo/Sets/HuntersHaven/Props/Textures/T_HH_Pancakes_3DSign_01_N
  }
  n6 = Transform(None=Multiply(
    A=MaterialFunctionCall(BaseNormal=@n4.PS, AdditionalNormal=@n5.RGB) {materialFunction=&BlendAngleCorrectedNormals}.Result
    B=TwoSidedSign()
   ))
  n7 = VertexNormalWS()
  n8 = Fresnel(
   ExponentIn=ScalarParameter() {defaultValue=3, parameterName=Fresnell}
   Normal=ShadingPathSwitch(Default=Normalize(VectorInput=@n6), Forward=@n7, Mobile=@n7)
  ) {baseReflectFraction=0}
  n9 = VectorParameter() {parameterName=EmissiveColor, defaultValue=<rgba 0.4 0.4 0.4 0>}
  n10 = TextureSampleParameter2D() {parameterName=Emissive, texture=/Game/Environments/Apollo/Sets/HuntersHaven/Props/Textures/T_HH_Pancakes_3DSign_01_D}
  n11 = WorldPosition()
  n12 = ComponentMask(None=ActorPositionWS()) {b=true}
  n13 = Multiply(A=ObjectRadius()) {constB=2}
  n14 = Time() {period=1}
  n15 = ScalarParameter() {defaultValue=0.5429, parameterName=SpeedRipple}
  n16 = LinearInterpolate(A=Multiply(A=@n13) {constB=-1}, B=@n13, Alpha=Frac(None=Multiply(A=@n14, B=@n15))) {constB=0}
  n17 = MaterialFunctionCall() {materialFunction=/Engine/Functions/Engine_MaterialFunctions02/WorldPositionOffset/ObjectScale}
  n18 = ScalarParameter() {defaultValue=30, parameterName=PulseLineRadius}
  n19 = Power(Base=SphereMask(A=ComponentMask(None=@n11.XYZ) {b=true}, B=Add(A=@n12, B=@n16), Radius=Multiply(A=@n17.Scale Z, B=@n18)) {hardnessPercent=0})
  n20 = Multiply(A=VectorParameter() {parameterName=PulseColor, defaultValue=<rgba 0.913 0.3365 1 1>}.RGB, B=@n19)
  n21 = Constant()
  Add(
   A=Add(A=Multiply(A=@n1.RGB, B=@n8), B=Multiply(A=@n9.RGB, B=@n10.RGB))
   B=FeatureLevelSwitch(Default=@n20, ES2=@n21, ES3_1=@n21, SM4_REMOVED=@n20)
  )
  n22 = ScalarParameter() {defaultValue=20, parameterName=PulseWPOMult}
  n23 = Multiply(A=@n19, B=Multiply(A=Multiply(A=@n22, B=@n17.Scale Z), B=VertexNormalWS()))
  n24 = Constant()
  Add(
   A=FeatureLevelSwitch(Default=@n23, ES2=@n24, ES3_1=@n24, SM4_REMOVED=@n23)
   B=RotateAboutAxis(
    NormalizedRotationAxis=@n2
    RotationAngle=@n3
    PivotPoint=MaterialFunctionCall() {materialFunction=/Game/ImpostorBaker/MaterialFunctions/ImpostorInfo_FromUVs}.Position
    Position=WorldPosition().XYZ
   )
  )
  n25 = ScalarParameter() {defaultValue=0.5, parameterName=Opacity}
  n26 = ScreenPosition()
  n27 = Sine(None=Multiply(A=ComponentMask(None=@n26.ViewportUV) {g=true}, B=ScalarParameter() {defaultValue=100, parameterName=ScanlineTile}))
  n28 = ScalarParameter() {defaultValue=1500, parameterName=FadeDistance}
  n29 = MaterialFunctionCall() {materialFunction=&ObjectScale}
  n30 = LinearInterpolate(
   A=Power(Base=ConstantBiasScale(None=@n27))
   Alpha=MaterialFunctionCall(Distance=Multiply(A=@n28, B=@n29.Scale Z)) {materialFunction=&Near_Camera_Fade}.Result
  ) {constA=0.4, constB=0.3}
  n31 = ScalarParameter() {defaultValue=-72.8749, parameterName=BiasSorting_Far}
  n32 = Constant()
  n33 = If(
   A=Add(A=PixelDepth().R, B=Multiply(A=@n29.Scale Z, B=@n31))
   B=ComponentMask(None=SceneTexture() {sceneTextureId=PPI_CustomDepth}.Color) {r=true}
   "A > B"=@n32
   "A == B"=@n32
   "A < B"=Constant() {r=1}
  )
  n34 = ScalarParameter() {defaultValue=1, parameterName=OopacityMobile}
  n35 = ObjectPositionWS()
  n36 = CameraPositionWS()
  n37 = Divide(
   A=Add(A=Distance(A=@n35, B=@n36), B=ScalarParameter() {defaultValue=-759.463, parameterName=MobileTwosidedFadeOffset})
   B=ScalarParameter() {defaultValue=100, parameterName=MobileTwosidedFade}
  )
  n38 = LinearInterpolate(
   A=LinearInterpolate(A=@n34, B=ScalarParameter() {parameterName=OopacityMobilebackface}, Alpha=Saturate(None=@n37))
   B=@n34
   Alpha=Saturate(None=TwoSidedSign())
  )
  n39 = Multiply(A=Clamp(None=Add(A=@n25, B=@n30) {constB=0.2}) {maxDefault=2}, B=FeatureLevelSwitch(Default=@n33, ES2=@n38, ES3_1=@n38, SM4_REMOVED=@n33))
  LinearInterpolate(A=@n39, B=Multiply(A=ParticleColor().A, B=@n39), Alpha=ScalarParameter() {parameterName=UseParticleColor})
 ]
}
