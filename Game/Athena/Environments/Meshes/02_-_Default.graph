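# Default Lit material in the Athena environment meshes folder, apparently a basic placeholder or import-generated material, judging by its "02 - Default" name.
# Its only node samples the T_Foundations texture, with no connection to any material output shown.
# Usage flags enable it for Nanite, skeletal meshes and mesh deformers.
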Material /Game/Athena/Environments/Meshes/02_-_Default
graph {
 properties {
  bUsedWithMeshDeformer=true
  bUsedWithNanite=true
  bUsedWithSkeletalMesh=true
  pixelDepthOffsetMode=PDOM_Legacy
  usedShadingModels=Default Lit
 }
 nodes [TextureSample() {texture=&/T_Foundations}]
}
